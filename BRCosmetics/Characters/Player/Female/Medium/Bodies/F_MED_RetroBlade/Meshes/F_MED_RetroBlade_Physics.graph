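# Physics asset for the female medium RetroBlade cosmetic body, covering only the right collar and its neighboring upper-body bones.
# Capsules on spine_05, the right clavicle and upper arm approximate the body around the collar's large box, which hangs off a tiny base box.
# The dynamic collar has narrow swing and twist limits to its base, while the base, clavicle and arm constraints leave angular motion free.
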PhysicsAsset /BRCosmetics/Characters/Player/Female/Medium/Bodies/F_MED_RetroBlade/Meshes/F_MED_RetroBlade_Physics
bodies {
 Base_collar_r [{shapeType=Box, extentX=1.01, extentY=1.01, extentZ=1.01, shapeName=Base_collar_r_box, rotation=<rot 0 -90 0>}]
 Dyn_collar_r_01 [{
   shapeType=Box
   extentX=17.1427
   extentY=3.4154
   extentZ=15.7308
   shapeName=Dyn_collar_r_01_box
   center=<8.7034 -3.6467 0.8887>
   rotation=<rot -19.4896 -169.922 -85.307>
  }]
 clavicle_r [{shapeType=Capsule, length=6, radius=4, shapeName=clavicle_r_capsule, center=<-6.0971 1.5706 0>, rotation=<rot -90 -0.0001 0>}]
 spine_05 [{shapeType=Capsule, length=6.6832, radius=6.4036}]
 upperarm_r [{shapeType=Capsule, length=13.7403, radius=4.7765, shapeName=upperarm_r_capsule, center=<-12.4838 0.5654 0>, rotation=<rot -90 0 0>}]
}
constraints [
 {
  bone1Name=Base_collar_r
  bone2Name=spine_05
  swing1LimitDegrees=45
  swing1Motion=Free
  swing2LimitDegrees=45
  swing2Motion=Free
  twistLimitDegrees=45
  twistMotion=Free
 }
 {bone1Name=Dyn_collar_r_01, bone2Name=Base_collar_r, swing1LimitDegrees=7, swing2LimitDegrees=7, twistLimitDegrees=7}
 {
  bone1Name=clavicle_r
  bone2Name=spine_05
  swing1LimitDegrees=45
  swing1Motion=Free
  swing2LimitDegrees=45
  swing2Motion=Free
  twistLimitDegrees=45
  twistMotion=Free
 }
 {
  bone1Name=upperarm_r
  bone2Name=clavicle_r
  swing1LimitDegrees=45
  swing1Motion=Free
  swing2LimitDegrees=45
  swing2Motion=Free
  twistLimitDegrees=45
  twistMotion=Free
 }
]
