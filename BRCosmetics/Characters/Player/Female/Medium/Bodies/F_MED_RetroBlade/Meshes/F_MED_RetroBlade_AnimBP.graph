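# Post-process style AnimBlueprint for the F_MED_RetroBlade female cosmetic body that copies the parent mesh's pose and adds secondary motion and corrective deformation.
# The pose runs through a Control Rig fed by player-state flags, RigidBody physics, chained AnimDynamics for amulet and ropes, weighted kneepad and joint constraints, then additive right-collar offsets.
# The EventGraph derives isMelee for the rig, plus simulation-space and weapon-aiming variables the AnimGraph never reads.
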AnimBlueprint /BRCosmetics/Characters/Player/Female/Medium/Bodies/F_MED_RetroBlade/Meshes/F_MED_RetroBlade_AnimBP
target_skeleton=&/F_MED_RetroBlade_Skeleton
graphs [
 AnimGraph {nodes [
   n1 = Control Rig!(
    Source=Copy Pose From Mesh() {bUseAttachedParent=true}
    Alpha
    IsPlaying:bool=Get bIsPlayingEmote()
    GetiIsSkyDIve:bool=Get bIsSkyDiving()
    IsTargeting:bool=Get bIsTargeting()
    GetIsFronEnd:bool=Get bIsInFrontEnd()
    IsMelee:bool=Get isMelee() -> (isMelee:bool)
   ) -> (Pose) {controlRigAssetReference {blueprintRigClass=&/F_MED_RetroBlade_CtrlRig.F_MED_RetroBlade_CtrlRig_C}}
   n2 = RigidBody(ComponentPose=Local To Component(LocalPose=@n1)) {
    bOverrideWorldGravity=true
    overridePhysicsAsset=&/F_MED_RetroBlade_Physics
    simulationSpace=BaseBoneSpace
    baseBoneRef {boneName=spine_05}
    overrideWorldGravity {z=-980}
    simSpaceSettings {worldAlpha=1}
   }
   n3 = Anim Dynamics - Bone: Dyn_Amilet_01!(SimulationSpace:AnimPhysSimSpaceType Enum, ComponentPose=@n2, Alpha=1, ExternalForce="0,0,0") -> (Pose) {
    angularDampingOverride=0.9999
    angularSpringConstant=500
    bAngularSpring=true
    bOverrideAngularDamping=true
    bUseGravityOverride=true
    bUsePlanarLimit=false
    bUseSphericalLimits=true
    linearDampingOverride=0.9999
    linearSpringConstant=500
    simulationSpace=RootRelative
    alphaBoolBlend {blendOption=Linear}
    boundBone {boneName=Dyn_Amilet_01}
    componentAppliedLinearAccClamp=<5000 5000 5000>
    componentLinearAccScale=<0.5 0.5 0.5>
    componentLinearVelScale=<0.5 0.5 0.5>
    gravityOverride {z=-980}
    physicsBodyDefinitions [{
      collisionType=CustomSphere
      sphereCollisionRadius=1.3
      boundBone {boneName=Dyn_Amilet_01}
      boxExtents=<15 15 15>
      constraintSetup {coneAngle=1, angularLimitsMax=<2 10 30>, angularLimitsMin=<-2 -10 1>, angularTarget=<1 0 0>}
      localJointOffset=<0.7 0 0>
     }]
    sphericalLimits [{limitRadius=3, drivingBone {boneName=neck_01}, sphereLocalOffset=<-6.7194 1.1698 4.9371>}]
   }
   n4 = Anim Dynamics - Bone: Dyn_Amulet_02!(SimulationSpace:AnimPhysSimSpaceType Enum, ComponentPose=@n3, Alpha=1, ExternalForce="0,0,0") -> (Pose) {
    angularDampingOverride=0.9999
    angularSpringConstant=500
    bAngularSpring=true
    bOverrideAngularDamping=true
    bUseGravityOverride=true
    bUsePlanarLimit=false
    bUseSphericalLimits=true
    linearDampingOverride=0.9999
    linearSpringConstant=500
    simulationSpace=RootRelative
    alphaBoolBlend {blendOption=Linear}
    boundBone {boneName=Dyn_Amulet_02}
    componentAppliedLinearAccClamp=<5000 5000 5000>
    componentLinearAccScale=<0.5 0.5 0.5>
    componentLinearVelScale=<0.5 0.5 0.5>
    gravityOverride {z=-980}
    physicsBodyDefinitions [{
      collisionType=CustomSphere
      sphereCollisionRadius=1.3
      boundBone {boneName=Dyn_Amulet_02}
      boxExtents=<15 15 15>
      constraintSetup {angularLimitsMax=<2 15 33>, angularLimitsMin=<-2 -15 1>, angularTarget=<1 0 0>}
      localJointOffset=<1 0 0>
     }]
    sphericalLimits [{limitRadius=3, drivingBone {boneName=neck_01}, sphereLocalOffset=<-6.7194 1.1698 4.9371>}]
   }
   n5 = Anim Dynamics - Bone: Dyn_rope_l!(SimulationSpace:AnimPhysSimSpaceType Enum, ComponentPose=@n4, Alpha=1, ExternalForce="0,0,0") -> (Pose) {
    angularDampingOverride=0.9999
    angularSpringConstant=500
    bAngularSpring=true
    bOverrideAngularDamping=true
    bUseGravityOverride=true
    bUsePlanarLimit=false
    bUseSphericalLimits=true
    linearDampingOverride=0.9999
    linearSpringConstant=500
    simulationSpace=RootRelative
    alphaBoolBlend {blendOption=Linear}
    boundBone {boneName=Dyn_rope_l}
    componentAppliedLinearAccClamp=<5000 5000 5000>
    componentLinearAccScale=<0.5 0.5 0.5>
    componentLinearVelScale=<0.5 0.5 0.5>
    gravityOverride {z=-980}
    physicsBodyDefinitions [{
      collisionType=CustomSphere
      sphereCollisionRadius=1.3
      boundBone {boneName=Dyn_rope_l}
      boxExtents=<15 15 15>
      constraintSetup {angularLimitsMax=<2 15 33>, angularLimitsMin=<-2 -15 1>, angularTarget=<1 0 0>}
      localJointOffset=<2 0 0>
     }]
    sphericalLimits [{limitRadius=3, drivingBone {boneName=neck_01}, sphereLocalOffset=<-6.7194 1.1698 4.9371>}]
   }
   n6 = Anim Dynamics - Bone: Dyn_rope_r!(SimulationSpace:AnimPhysSimSpaceType Enum, ComponentPose=@n5, Alpha=1, ExternalForce="0,0,0") -> (Pose) {
    angularDampingOverride=0.9999
    angularSpringConstant=500
    bAngularSpring=true
    bOverrideAngularDamping=true
    bUseGravityOverride=true
    bUsePlanarLimit=false
    bUseSphericalLimits=true
    linearDampingOverride=0.9999
    linearSpringConstant=500
    simulationSpace=RootRelative
    alphaBoolBlend {blendOption=Linear}
    boundBone {boneName=Dyn_rope_r}
    componentAppliedLinearAccClamp=<5000 5000 5000>
    componentLinearAccScale=<0.5 0.5 0.5>
    componentLinearVelScale=<0.5 0.5 0.5>
    gravityOverride {z=-980}
    physicsBodyDefinitions [{
      collisionType=CustomSphere
      sphereCollisionRadius=1.3
      boundBone {boneName=Dyn_rope_r}
      boxExtents=<15 15 15>
      constraintSetup {angularLimitsMax=<2 15 33>, angularLimitsMin=<-2 -15 1>, angularTarget=<1 0 0>}
      localJointOffset=<2 0 0>
     }]
    sphericalLimits [{limitRadius=3, drivingBone {boneName=neck_01}, sphereLocalOffset=<-6.7194 1.1698 4.9371>}]
   }
   n7 = "Constraint - Corr_kneepad (PPP)"!(ComponentPose=@n6, Alpha=1, ConstraintWeights_0=0.2, ConstraintWeights_1:float=0.5, ConstraintWeights_2:float=0.3) -> (Pose) {
    boneToModify {boneName=Corr_kneepad}
    constraintSetup [{targetBone {boneName=thigh_l}}, {targetBone {boneName=calf_l}}, {targetBone {boneName=deform_knee_out_l}}]
    constraintWeights [1,1,1]
   }
   n8 = "Constraint - joint55 (PPPPPP)"!(
    ComponentPose=@n7
    Alpha=1
    ConstraintWeights_0=0.0003
    ConstraintWeights_1:float=0
    ConstraintWeights_2:float=0.0002
    ConstraintWeights_3:float=0.4689
    ConstraintWeights_4:float=0.0114
    ConstraintWeights_5:float=0.5192
   ) -> (Pose) {
    boneToModify {boneName=joint55}
    constraintSetup [
     {targetBone {boneName=spine_02}}
     {targetBone {boneName=spine_04}}
     {targetBone {boneName=spine_01}}
     {targetBone {boneName=pelvis}}
     {targetBone {boneName=thigh_twist_01_l}}
     {targetBone {boneName=Corr_belt_l}}
    ]
    constraintWeights [1,1,1,1,1,1]
   }
   n9 = "Constraint - joint44 (PPPP)"!(
    ComponentPose=@n8
    Alpha=1
    ConstraintWeights_0=0.0046
    ConstraintWeights_1:float=0.5959
    ConstraintWeights_2:float=0.0127
    ConstraintWeights_3:float=0.3867
   ) -> (Pose) {
    boneToModify {boneName=joint44}
    constraintSetup [
     {targetBone {boneName=spine_02}}
     {targetBone {boneName=pelvis}}
     {targetBone {boneName=thigh_twist_01_l}}
     {targetBone {boneName=Corr_belt_l}}
    ]
    constraintWeights [1,1,1,1]
   }
   n10 = "Constraint - joint33 (PPPPP)"!(
    ComponentPose=@n9
    Alpha=1
    ConstraintWeights_0=0.0005
    ConstraintWeights_1:float=0.0021
    ConstraintWeights_2:float=0.7497
    ConstraintWeights_3:float=0.0118
    ConstraintWeights_4:float=0.2358
   ) -> (Pose) {
    boneToModify {boneName=joint33}
    constraintSetup [
     {targetBone {boneName=upperarm_twist_02_l}}
     {targetBone {boneName=spine_02}}
     {targetBone {boneName=pelvis}}
     {targetBone {boneName=thigh_twist_01_l}}
     {targetBone {boneName=Corr_belt_l}}
    ]
    constraintWeights [1,1,1,1,1]
   }
   n11 = "Constraint - joint22 (PPPP)"!(
    ComponentPose=@n10
    Alpha=1
    ConstraintWeights_0=0.8767
    ConstraintWeights_1:float=0.0208
    ConstraintWeights_2:float=0.0006
    ConstraintWeights_3:float=0.1018
   ) -> (Pose) {
    boneToModify {boneName=joint22}
    constraintSetup [
     {targetBone {boneName=pelvis}}
     {targetBone {boneName=thigh_twist_01_l}}
     {targetBone {boneName=thigh_l}}
     {targetBone {boneName=Corr_belt_l}}
    ]
    constraintWeights [1,1,1,1]
   }
   n12 = "Constraint - joint11 (PPPP)"!(
    ComponentPose=@n11
    Alpha=1
    ConstraintWeights_0=0.9469
    ConstraintWeights_1:float=0.0257
    ConstraintWeights_2:float=0.0011
    ConstraintWeights_3:float=0.0262
   ) -> (Pose) {
    boneToModify {boneName=joint11}
    constraintSetup [
     {targetBone {boneName=pelvis}}
     {targetBone {boneName=thigh_twist_01_l}}
     {targetBone {boneName=thigh_l}}
     {targetBone {boneName=Corr_belt_l}}
    ]
    constraintWeights [1,1,1,1]
   }
   n13 = "Transform (Modify) Bone - Bone: Base_collar_r"!(ComponentPose=@n12, bAlphaBoolEnabled:bool=True) -> (Pose) {
    alphaInputType=Bool
    rotationMode=BMM_Additive
    rotationSpace=BCS_BoneSpace
    translationMode=BMM_Additive
    translationSpace=BCS_BoneSpace
    boneToModify {boneName=Base_collar_r}
    rotation {pitch=10, roll=-10}
    translation {y=-3, z=1.5}
   }
   n14 = "Transform (Modify) Bone - Bone: Base_collar_r"(ComponentPose=@n13, bAlphaBoolEnabled=True) {
    alphaInputType=Bool
    rotationMode=BMM_Additive
    rotationSpace=BCS_BoneSpace
    translationMode=BMM_Additive
    translationSpace=BCS_BoneSpace
    boneToModify {boneName=Base_collar_r}
    rotation {pitch=15, roll=-15}
    translation {y=-3, z=1.5}
   }
   Output Pose(Result=Component To Local(ComponentPose=@n14)) {layerGroup=None}
  ]}
 EventGraph {nodes [
   Event BlueprintUpdateAnimation()
   > n1 = Sequence!(execute) -> (then_0, then_1, then_2:Exec)
   n2 = Cast To FortPlayerAnimInstance(execute=@n1.then_2, Object=GetAnimInstance(self=GetOwnerSkeletalMeshComponent()))
   > n3 = Switch on EFortWeaponCoreAnimation!(execute, Selection:EFortWeaponCoreAnimation Enum=Get WeaponCoreAnim(self=@n2)) -> (
    Melee:Exec
    Pistol:Exec
    Shotgun:Exec
    PaperBlueprint:Exec
    Rifle:Exec
    MeleeOneHand:Exec
    MachinePistol:Exec
    RocketLauncher:Exec
    GrenadeLauncher:Exec
    GoingCommando:Exec
    AssaultRifle:Exec
    TacticalShotgun:Exec
    SniperRifle:Exec
    TrapPlacement:Exec
    ShoulderLauncher:Exec
    AbilityDecoTool:Exec
    Crossbow:Exec
    C4:Exec
    RemoteControl:Exec
    DualWield:Exec
    AR_BullPup:Exec
    AR_ForwardGrip:Exec
    MedPackPaddles:Exec
    SMG_P90:Exec
    AR_DrumGun:Exec
    Consumable_Small:Exec
    Consumable_Large:Exec
    Balloon:Exec
    MountedTurret:Exec
    CreativeTool:Exec
    ExplosiveBow:Exec
    AshtonIndigo:Exec
    AshtonChicago:Exec
    MeleeDualWield:Exec
    Unarmed:Exec
   )
   Set isMelee(
    execute=[
     @n3.AR_BullPup
     @n3.AR_DrumGun
     @n3.AR_ForwardGrip
     @n3.AbilityDecoTool
     @n3.AshtonChicago
     @n3.AshtonIndigo
     @n3.AssaultRifle
     @n3.Balloon
     @n3.C4
     @n3.Consumable_Large
     @n3.Consumable_Small
     @n3.CreativeTool
     @n3.Crossbow
     @n3.DualWield
     @n3.ExplosiveBow
     @n3.GoingCommando
     @n3.GrenadeLauncher
     @n3.MachinePistol
     @n3.MedPackPaddles
     @n3.MeleeDualWield
     @n3.MeleeOneHand
     @n3.MountedTurret
     @n3.PaperBlueprint
     @n3.Pistol
     @n3.RemoteControl
     @n3.Rifle
     @n3.RocketLauncher
     @n3.SMG_P90
     @n3.Shotgun
     @n3.ShoulderLauncher
     @n3.SniperRifle
     @n3.TacticalShotgun
     @n3.TrapPlacement
     @n3.Unarmed
    ]
    isMelee:bool=false
   ) -> (Output_Get:bool)
   Set isMelee(execute=@n3.Melee, isMelee=true)
   n4 = Branch(execute=@n1.then_0, Condition=Get bIsInFrontEnd())
   > Set SimlationSpace(SimlationSpace:AnimPhysSimSpaceType Enum=World) -> (Output_Get:AnimPhysSimSpaceType Enum)
   Set SimlationSpace(execute=@n4.else, SimlationSpace=Component)
   n5 = Get WeaponCoreAnim!() -> (WeaponCoreAnim)
   Set isAssaultRifle(
    execute=@n1.then_1
    isAssaultRifle:bool=AND Boolean(A="Equal (Enum)"(A:EFortWeaponCoreAnimation Enum=@n5, B:EFortWeaponCoreAnimation Enum=AssaultRifle), B=Get bIsTargeting())
   ) -> (Output_Get:bool)
   > Set isSMG(isSMG:bool=AND Boolean(A=Get bIsTargeting(), B="Equal (Enum)"(A=@n5, B=SMG_P90))) -> (Output_Get:bool)
   Set isPistol(isPistol=false)
  ]}
]
variables [
 isAssaultRifle {type=boolean}
 isMelee {type=boolean}
 isPistol {type=boolean}
 isSMG {type=boolean}
 SimlationSpace {type=AnimPhysSimSpaceType}
]
