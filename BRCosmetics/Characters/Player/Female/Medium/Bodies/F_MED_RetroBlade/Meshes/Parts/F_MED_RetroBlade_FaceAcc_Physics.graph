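# Physics asset for the face accessory mesh of the F_MED_RetroBlade female player cosmetic, setting up physics for its hair.
# Five Dyn_hair capsules, mostly widening toward the tip, chain from Base_hair, which is locked to the head.
# Hair joints lock twist and limit swing, very tightly at the root and loosely further down.
# Head, neck, upper spine and upper-arm capsules supply body shapes, with the head locked to the neck and other body joints free.
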PhysicsAsset /BRCosmetics/Characters/Player/Female/Medium/Bodies/F_MED_RetroBlade/Meshes/Parts/F_MED_RetroBlade_FaceAcc_Physics
bodies {
 Base_hair [{shapeType=Capsule, length=0.505, radius=0.505, shapeName=Base_hair_capsule, rotation=<rot 0 -90 0>}]
 Dyn_hair_1 [{shapeType=Capsule, length=7.1182, radius=3.9628, shapeName=Dyn_hair_1_capsule, center=<0 -0.4883 0>}]
 Dyn_hair_2 [{shapeType=Capsule, length=5.2436, radius=4.6994, shapeName=Dyn_hair_2_capsule, center=<0 -0.4883 0>}]
 Dyn_hair_3 [{shapeType=Capsule, length=8.1251, radius=5.3383, shapeName=Dyn_hair_3_capsule, center=<0.9229 0.6116 0>, rotation=<rot 0 50 0>}]
 Dyn_hair_4 [{shapeType=Capsule, length=10.3293, radius=6.2352, shapeName=Dyn_hair_4_capsule, center=<2.7542 -0.4883 0>}]
 Dyn_hair_5 [{shapeType=Capsule, length=9.0962, radius=8.8254, shapeName=Dyn_hair_5_capsule, center=<8.4668 -2.603 0>}]
 head [{shapeType=Capsule, length=2.9586, radius=8.6646, shapeName=head_capsule, center=<7.1872 2.1182 0.1001>, rotation=<rot 1.8782 -90 -92.8199>}]
 neck_02 [{shapeType=Capsule, length=25.005, radius=6.5, shapeName=neck_02_capsule, center=<-2.741 0.6641 -0.2398>, rotation=<rot 0 -90.0002 -5>}]
 spine_04 [{shapeType=Capsule, length=11, radius=11, shapeName=spine_04_capsule, center=<0 3.1973 0>, rotation=<rot 0 -90 0>}]
 spine_05 [{shapeType=Capsule, length=11, radius=11, shapeName=spine_05_capsule, center=<0.8019 3.8278 0>, rotation=<rot 0 -90 0>}]
 upperarm_l [{shapeType=Capsule, length=23.5, radius=7.5, shapeName=upperarm_l_capsule, center=<14.415 0 2.5418>, rotation=<rot -0.0001 -89.9999 80>}]
 upperarm_r [{shapeType=Capsule, length=23.5, radius=7.5, shapeName=upperarm_r_capsule, center=<-14.415 0 -2.5418>, rotation=<rot -0.0001 -89.9999 -100>}]
}
constraints [
 {
  bone1Name=head
  bone2Name=neck_02
  swing1LimitDegrees=45
  swing1Motion=Locked
  swing2LimitDegrees=45
  swing2Motion=Locked
  twistLimitDegrees=45
  twistMotion=Locked
 }
 {
  bone1Name=neck_02
  bone2Name=spine_05
  swing1LimitDegrees=45
  swing1Motion=Free
  swing2LimitDegrees=45
  swing2Motion=Free
  twistLimitDegrees=45
  twistMotion=Free
 }
 {
  bone1Name=spine_05
  bone2Name=spine_04
  swing1LimitDegrees=45
  swing1Motion=Free
  swing2LimitDegrees=45
  swing2Motion=Free
  twistLimitDegrees=45
  twistMotion=Free
 }
 {
  bone1Name=Base_hair
  bone2Name=head
  swing1LimitDegrees=45
  swing1Motion=Locked
  swing2LimitDegrees=45
  swing2Motion=Locked
  twistLimitDegrees=45
  twistMotion=Locked
 }
 {bone1Name=Dyn_hair_1, bone2Name=Base_hair, swing1LimitDegrees=3, swing2LimitDegrees=3, twistLimitDegrees=45, twistMotion=Locked}
 {bone1Name=Dyn_hair_2, bone2Name=Dyn_hair_1, swing1LimitDegrees=50, swing2LimitDegrees=45, twistLimitDegrees=45, twistMotion=Locked}
 {bone1Name=Dyn_hair_3, bone2Name=Dyn_hair_2, swing1LimitDegrees=50, swing2LimitDegrees=45, twistLimitDegrees=45, twistMotion=Locked}
 {bone1Name=Dyn_hair_4, bone2Name=Dyn_hair_3, swing1LimitDegrees=50, swing2LimitDegrees=45, twistLimitDegrees=45, twistMotion=Locked}
 {bone1Name=Dyn_hair_5, bone2Name=Dyn_hair_4, swing1LimitDegrees=50, swing2LimitDegrees=45, twistLimitDegrees=45, twistMotion=Locked}
 {
  bone1Name=upperarm_l
  bone2Name=spine_05
  swing1LimitDegrees=45
  swing1Motion=Free
  swing2LimitDegrees=45
  swing2Motion=Free
  twistLimitDegrees=45
  twistMotion=Free
 }
 {
  bone1Name=upperarm_r
  bone2Name=spine_05
  swing1LimitDegrees=45
  swing1Motion=Free
  swing2LimitDegrees=45
  swing2Motion=Free
  twistLimitDegrees=45
  twistMotion=Free
 }
]
