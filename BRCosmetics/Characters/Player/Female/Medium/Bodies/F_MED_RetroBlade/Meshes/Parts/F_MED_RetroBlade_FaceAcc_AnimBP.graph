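# Post-process-style AnimBP adding secondary motion to hair and earrings on the female RetroBlade face accessory mesh.
# Hair uses head-relative RigidBody simulation or constraint fallback, selected by bRigidBodyEnabled, then gets corrective bone rotations driven by head-to-neck rotation when backbling or cape is worn.
# Earrings use spring AnimDynamics with neck sphere limits, plus base-bone offsets.
# The event graph updates head rotations and state flags, tilting simulation gravity during slides and tactical sprints.
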AnimBlueprint /BRCosmetics/Characters/Player/Female/Medium/Bodies/F_MED_RetroBlade/Meshes/Parts/F_MED_RetroBlade_FaceAcc_AnimBP
target_skeleton=&/F_MED_RetroBlade_FaceAcc_Skeleton
graphs [
 AnimGraph {nodes [
   n1 = Get head_fwdBwd_rot() -> (head_fwdBwd_rot:double)
   n2 = "Transform (Modify) Bone - Bone: Dyn_hair_2"(
    ComponentPose=Local To Component(LocalPose=Use cached pose 'Rban'())
    Alpha=@n1
    Translation="0.015,0,0.678"
    Rotation="0,-20,0"
    Scale="1,1,1"
   ) {
    rotationMode=BMM_Additive
    rotationSpace=BCS_BoneSpace
    translationMode=BMM_Additive
    translationSpace=BCS_BoneSpace
    alphaBoolBlend {blendOption=Linear}
    alphaScaleBiasClamp {bMapRange=true, inRange {max=0, min=-25}, outRange {max=0, min=1}}
    boneToModify {boneName=Dyn_hair_2}
    rotation {yaw=-20}
    translation {x=0.015, z=0.678}
   }
   n3 = "Transform (Modify) Bone - Bone: Dyn_hair_4"(ComponentPose=@n2, Alpha=@n1, Translation="0,0,0", Rotation="-0.118,15,0.033", Scale="1,1,1") {
    rotationMode=BMM_Additive
    rotationSpace=BCS_BoneSpace
    translationMode=BMM_Additive
    translationSpace=BCS_BoneSpace
    alphaBoolBlend {blendOption=Linear}
    alphaScaleBiasClamp {bMapRange=true, inRange {max=0, min=-25}, outRange {max=0, min=1}}
    boneToModify {boneName=Dyn_hair_4}
    rotation=<rot -0.118 15 0.033>
   }
   n4 = "Transform (Modify) Bone - Bone: Dyn_hair_5"(ComponentPose=@n3, Alpha=@n1, Translation="0,0,0", Rotation="-0.118,-20,0.033", Scale="1,1,1") {
    rotationMode=BMM_Additive
    rotationSpace=BCS_BoneSpace
    translationMode=BMM_Additive
    translationSpace=BCS_BoneSpace
    alphaBoolBlend {blendOption=Linear}
    alphaScaleBiasClamp {bMapRange=true, inRange {max=0, min=-25}, outRange {max=0, min=1}}
    boneToModify {boneName=Dyn_hair_5}
    rotation=<rot -0.118 -20 0.033>
   }
   n5 = Anim Dynamics - Bone: Dyn_earring_l(ComponentPose=@n4, Alpha=1, ExternalForce="0,0,0") {
    angularDampingOverride=0.9999
    angularSpringConstant=500
    bAngularSpring=true
    bOverrideAngularDamping=true
    bUseGravityOverride=true
    bUsePlanarLimit=false
    bUseSphericalLimits=true
    linearDampingOverride=0.9999
    linearSpringConstant=500
    simulationSpace=RootRelative
    alphaBoolBlend {blendOption=Linear}
    boundBone {boneName=Dyn_earring_l}
    componentAppliedLinearAccClamp=<5000 5000 5000>
    componentLinearAccScale=<0.5 0.5 0.5>
    componentLinearVelScale=<0.5 0.5 0.5>
    gravityOverride {z=-980}
    physicsBodyDefinitions [{
      sphereCollisionRadius=1.3
      boundBone {boneName=Dyn_earring_l}
      boxExtents=<9 9 9>
      constraintSetup {angularLimitsMax=<2 7 15>, angularLimitsMin=<-2 -7 -15>, angularTarget=<1 0 0>}
      localJointOffset=<1 0 0>
     }]
    simSpaceSettings {maxAngularAcceleration=30, maxAngularVelocity=6, simSpaceAngularAlpha=1}
    sphericalLimits [{limitRadius=3, drivingBone {boneName=neck_01}, sphereLocalOffset=<-6.7194 1.1698 4.9371>}]
   }
   n6 = Anim Dynamics - Bone: Dyn_earring_r(ComponentPose=@n5, Alpha=1, ExternalForce="0,0,0") {
    angularDampingOverride=0.9999
    angularSpringConstant=500
    bAngularSpring=true
    bOverrideAngularDamping=true
    bUseGravityOverride=true
    bUsePlanarLimit=false
    bUseSphericalLimits=true
    linearDampingOverride=0.9999
    linearSpringConstant=500
    simulationSpace=RootRelative
    alphaBoolBlend {blendOption=Linear}
    boundBone {boneName=Dyn_earring_r}
    componentAppliedLinearAccClamp=<5000 5000 5000>
    componentLinearAccScale=<0.5 0.5 0.5>
    componentLinearVelScale=<0.5 0.5 0.5>
    gravityOverride {z=-980}
    physicsBodyDefinitions [{
      sphereCollisionRadius=1.3
      boundBone {boneName=Dyn_earring_r}
      boxExtents=<9 9 9>
      constraintSetup {angularLimitsMax=<2 7 15>, angularLimitsMin=<-2 -7 -15>, angularTarget=<1 0 0>}
      localJointOffset=<1 0 0>
     }]
    simSpaceSettings {maxAngularAcceleration=30, maxAngularVelocity=6, simSpaceAngularAlpha=1}
    sphericalLimits [{limitRadius=3, drivingBone {boneName=neck_01}, sphereLocalOffset=<-6.7194 1.1698 4.9371>}]
   }
   n7 = "Transform (Modify) Bone - Bone: Base_earring_r"!(ComponentPose=@n6, bAlphaBoolEnabled:bool=True) -> (Pose) {
    alphaInputType=Bool
    rotationMode=BMM_Additive
    rotationSpace=BCS_BoneSpace
    alphaScaleBiasClamp {bInterpResult=true, interpSpeedDecreasing=15, interpSpeedIncreasing=15}
    boneToModify {boneName=Base_earring_r}
    rotation {pitch=13, yaw=-13}
   }
   n8 = "Transform (Modify) Bone - Bone: Base_earring_r"!(ComponentPose=@n7, Alpha:float=1) -> (Pose) {
    rotationMode=BMM_Additive
    rotationSpace=BCS_BoneSpace
    alphaScaleBiasClamp {bInterpResult=true, interpSpeedDecreasing=15, interpSpeedIncreasing=15}
    boneToModify {boneName=Base_earring_r}
    rotation {pitch=13, yaw=-13}
   }
   n9 = "Transform (Modify) Bone - Bone: Base_earring_l"!(ComponentPose=@n8, Alpha=1) -> (Pose) {
    rotationMode=BMM_Additive
    rotationSpace=BCS_BoneSpace
    alphaScaleBiasClamp {bInterpResult=true, interpSpeedDecreasing=15, interpSpeedIncreasing=15}
    boneToModify {boneName=Base_earring_l}
    rotation=<rot 7 25 -20>
   }
   Output Pose(Result=Component To Local(ComponentPose=@n9)) {layerGroup=None}
   n10 = RigidBody!(
    ComponentPose=Local To Component(LocalPose=Copy Pose From Mesh() {bUseAttachedParent=true})
    Alpha
    OverrideWorldGravity:Vector=Get Override World Gravity() -> (Override World Gravity:Vector)
    ExternalForce
   ) -> (Pose) {
    bClampLinearTranslationLimitToRefPose=true
    bOverrideWorldGravity=true
    overlapChannel=ECC_WorldDynamic
    overridePhysicsAsset=&/F_MED_RetroBlade_FaceAcc_Physics
    simulationSpace=BaseBoneSpace
    alphaBoolBlend {blendOption=Linear}
    baseBoneRef {boneName=head}
    simSpaceSettings {
     dampingAlpha=0.3
     maxAngularAcceleration=250
     maxAngularVelocity=5
     maxLinearAcceleration=1000
     maxLinearVelocity=250
     velocityScaleZ=0.8
     worldAlpha=1
     externalLinearDragV=<2 1.5 1>
    }
   }
   n11 = "Transform (Modify) Bone - Bone: Base_hair"(ComponentPose=@n10, Alpha=1, Translation="0,0,0", Rotation="15,-7,15", Scale="1,1,1") {
    rotationMode=BMM_Additive
    rotationSpace=BCS_BoneSpace
    alphaScaleBiasClamp {bInterpResult=true, interpSpeedDecreasing=100}
    boneToModify {boneName=Base_hair}
    rotation=<rot 15 -7 15>
   }
   n12 = "Transform (Modify) Bone - Bone: Base_hair"!(ComponentPose=@n11, bAlphaBoolEnabled:bool=True, Translation="0,0,0", Rotation="0,-10,0", Scale="1,1,1") -> (Pose) {alphaInputType=Bool, rotationMode=BMM_Additive, rotationSpace=BCS_BoneSpace, boneToModify {boneName=Base_hair}, rotation {yaw=-10}}
   n13 = "Constraint - Dyn_hair_5 (PPP)"!(
    ComponentPose=Local To Component(LocalPose=Copy Pose From Mesh() {bUseAttachedParent=true})
    Alpha=1
    ConstraintWeights_0=0.5
    ConstraintWeights_1:float=0.5
    ConstraintWeights_2:float=0
   ) -> (Pose) {
    alphaBoolBlend {blendOption=Linear}
    boneToModify {boneName=Dyn_hair_5}
    constraintSetup [{targetBone {boneName=head}}, {targetBone {boneName=neck_01}}, {targetBone {boneName=spine_05}}]
    constraintWeights [1,1,1]
   }
   n14 = "Constraint - Dyn_hair_1 (PPP)"!(ComponentPose=@n13, Alpha=1, ConstraintWeights_0=0.5, ConstraintWeights_1:float=0.5, ConstraintWeights_2:float=0) -> (Pose) {
    alphaBoolBlend {blendOption=Linear}
    boneToModify {boneName=Dyn_hair_1}
    constraintSetup [{targetBone {boneName=head}}, {targetBone {boneName=neck_01}}, {targetBone {boneName=spine_05}}]
    constraintWeights [1,1,1]
   }
   Save cached pose 'Rban'(Pose=Blend Poses by bool(BlendPose_0=Component To Local(ComponentPose=@n12), BlendPose_1=Component To Local(ComponentPose=@n14), bActiveValue=Get bRigidBodyEnabled()) {blendType=Linear}) {cacheName=Rban}
  ]}
 EventGraph {nodes [
   Event BlueprintUpdateAnimation()
   > n1 = Sequence!(execute) -> (then_0, then_1, then_2:Exec, then_3:Exec, then_4:Exec)
   > n2 = Branch(Condition=Get bIsPlayingEmote())
   n3 = BreakTransform(InTransform=GetDeltaTransformFromRefPose(self=GetOwnerSkeletalMeshComponent(), BoneName=head, BaseName=spine_05))
   Set Head Z Rotate(execute=@n2.else, Head Z Rotate:double=MapRangeClamped(Value=BreakRotator(InRot=@n3.Rotation).Yaw, InRangeA=-10, InRangeB=-35)) -> (Output_Get:double)
   n4 = Get is Wearing Backbling or Cape() -> (is Wearing Backbling or Cape:bool)
   > n5 = Branch(Condition="Nearly Equal (Float)"(A="To Float (Boolean)"(InBool=@n4), B=1, ErrorTolerance=0.1))
   n6 = GetDeltaTransformFromRefPose(self=GetOwnerSkeletalMeshComponent(), BoneName=head, BaseName=neck_01)
   > Set head_fwdBwd_rot(head_fwdBwd_rot:double=BreakRotator(InRot=BreakTransform(InTransform=@n6).Rotation).Yaw) -> (Output_Get:double)
   Set head_fwdBwd_rot(execute=@n5.else, head_fwdBwd_rot=0)
   n7 = CheckForBackbling()
   n8 = NOT Boolean(A=@n7.bIsBackblingACape)
   n9 = NOT Boolean(A=@n7.bIsBackblingLong)
   Set is Wearing Backbling or Cape(
    execute=@n1.then_1
    is Wearing Backbling or Cape:bool=OR Boolean(A=AND Boolean!(A=@n7, B=@n8, C:bool=@n9) -> (ReturnValue), B=OR Boolean(A=@n7.bIsBackblingACape, B=@n7.bIsBackblingLong))
   ) -> (Output_Get:bool)
   Set DBNoORSky(execute=@n1.then_2, DBNoORSky:bool=OR Boolean(A=Get bIsDBNO(), B=Get bIsSkydiveFloating())) -> (Output_Get:bool)
   n10 = Get WeaponCoreAnim!() -> (WeaponCoreAnim)
   > Set IsPistol(IsPistol:bool=AND Boolean!(A="Equal (Enum)"(A:EFortWeaponCoreAnimation Enum=@n10, B:EFortWeaponCoreAnimation Enum=Pistol), B=Get bIsTargeting()) -> (ReturnValue)) -> (Output_Get:bool)
   n11 = Branch(execute=@n1.then_3, Condition=OR Boolean(A=Get bIsSliding(), B=Get bIsTacticalSprinting()))
   > Set Override World Gravity(Override World Gravity:Vector="0,-500,-800") -> (Output_Get:Vector)
   Set Override World Gravity(execute=@n11.else, Override World Gravity="0,0,-850")
   n12 = GetOwnerSkeletalMeshComponent()
   n13 = BreakRotator(InRot=BreakTransform(InTransform=GetDeltaTransformFromRefPose(self=@n12, BoneName=head, BaseName=neck_01)).Rotation)
   Set Head -X Rotate(execute=@n1.then_4, Head -X Rotate:double=MapRangeClamped(Value=@n13, InRangeA=-10, InRangeB=-35)) -> (Output_Get:double)
   > Set Head X Rotate(Head X Rotate:double=MapRangeClamped(Value=@n13, InRangeA=5, InRangeB=35)) -> (Output_Get:double)
  ]}
]
variables [
 DBNoORSky {type=boolean}
 Head -X Rotate {type=number}
 Head X Rotate {type=number}
 Head Z Rotate {type=number}
 head_fwdBwd_rot {type=number}
 is Wearing Backbling or Cape {type=boolean}
 IsPistol {type=boolean}
 Override World Gravity {type=Vector, default=<0 0 -600>}
]
